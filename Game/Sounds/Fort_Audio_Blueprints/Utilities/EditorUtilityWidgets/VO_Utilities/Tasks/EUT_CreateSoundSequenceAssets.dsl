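; Editor utility task that batch-creates FortSoundSequence assets from an array of sound sequence data.
; For each entry it creates a named asset in the destination folder, fills it via PopulateSoundSequenceData and adds it to a created-assets list.
; It checks for cancellation on each loop pass, and a cancel request deletes the assets created so far.
; The task title shows how many sequences are being created.
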(event Task|EventBeginExecution
  (bind _array_element (Utilities|Array|ForEachLoopwithBreak (Variables|Default|GetSound%20Sequence%20Data))
    (:LoopBody
      (if (Task|WasCancelRequested)
        _)
      (bind _output_get (Variables|Default|SetCurrent%20Sound%20Sequence%20Data _array_element))
      (bind _returnvalue (EditorScripting|AssetTools|GetAssetTools))
      (bind _returnvalue_2 (EditorScripting|AssetTools|CreateAsset _returnvalue (Variables|/Game/Sounds/Fort_Audio_Blueprints/Utilities/EditorUtilityWidgets/VO_Utilities/Data/FortSoundSequenceData.FortSoundSequenceData_C|GetAsset%20Name _output_get) (Variables|Default|GetDestination%20Folder%20Path) "/Script/FortniteGame.FortSoundSequence" 0 "None" false))
      (bind _asfort_sound_sequence (Utilities|Casting|CastToFortSoundSequence _returnvalue_2))
      (bind _output_get_2 (Variables|Default|SetCurrent%20Sound%20Sequence _asfort_sound_sequence))
      (AudioEditorFunctions|Dialogue|PopulateSoundSequenceData _output_get_2 (Variables|Default|GetCurrent%20Sound%20Sequence%20Data))
      (Utilities|Array|Add (Variables|Default|GetCreated%20Assets) (Variables|Default|GetCurrent%20Sound%20Sequence)))
    (:Completed
      _))
  (Task|FinishExecutingTask true))

(event Task|EventCancelRequested
  (EditorScripting|Asset|DeleteLoadedAssets (Variables|Default|GetCreated%20Assets) :!class /Script/EditorScriptingUtilities.EditorAssetLibrary))

(fn GetTaskTitleOverride
  (return (Utilities|Text|FormatText "NSLOCTEXT(\"[F81FA5F917F0963BEC0152F6F92A590E]\", \"69D0DBEB4D8564AB0E3D78BF4CFEF657\", \"Creating {num} Sound Sequences\")" (Utilities|Array|Length (Variables|Default|GetSound%20Sequence%20Data)))))
